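# Sound cue for the final impact when a candle is demolished.
# A random node picks one of three equally weighted candle impact waves.
# It plays through the impacts sound class with medium-sized object destroy attenuation and a reduced volume.
# It references the shared four-max object impact concurrency set and also carries its own stop-oldest concurrency settings.
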SoundCue /Game/Sounds/Fort_Impact_Sounds/Demolition/Candle/Demo_Candle_ImpactFinal_Cue
root=Random_63
nodes [
 {id=Random_63, children [WavePlayer_203, WavePlayer_204, WavePlayer_205], properties {weights [1,1,1]}}
 {id=WavePlayer_203, properties {soundWaveAssetPtr=&/Candle_ImpactFinal04}}
 {id=WavePlayer_204, properties {soundWaveAssetPtr=&/Candle_ImpactFinal01}}
 {id=WavePlayer_205, properties {soundWaveAssetPtr=&/Candle_ImpactFinal02}}
]
properties {
 attenuationSettings=/Game/Sounds/Attenuation/Fort_Demolition_Destroy_Medium_Sized_Object_Att
 maxDistance=4000
 soundClassObject=&Fort_Impacts_PSM
 volumeMultiplier=0.544
 concurrencyOverrides {maxCount=3, resolutionRule=StopOldest}
 concurrencySet [&CONC_Object_Impacts_FourMax]
}
