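# Sound cue for candle impact sounds in the demolition impact set.
# A random node picks one of three equally weighted candle impact waves and feeds it through a modulator for pitch and volume variation.
# It plays in the impacts sound class with small melee weapon attenuation and reduced volume, and a shared object-impact concurrency set limits simultaneous playback.
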SoundCue /Game/Sounds/Fort_Impact_Sounds/Demolition/Candle/Demo_Candle_Impact_Cue
root=Modulator_37
nodes [
 {id=Modulator_37, children [Random_62]}
 {id=Random_62, children [WavePlayer_200, WavePlayer_201, WavePlayer_202], properties {weights [1,1,1]}}
 {id=WavePlayer_200, properties {soundWaveAssetPtr=&/Candle_Impact02}}
 {id=WavePlayer_201, properties {soundWaveAssetPtr=&/Candle_Impact05}}
 {id=WavePlayer_202, properties {soundWaveAssetPtr=&/Candle_Impact04}}
]
properties {
 attenuationSettings=&Fort_Weapon_Melee_Sml_Att
 maxDistance=2800
 soundClassObject=&Fort_Impacts
 volumeMultiplier=0.68
 concurrencyOverrides {maxCount=3, resolutionRule=StopOldest}
 concurrencySet [&CONC_Object_Impacts_FourMax]
}
